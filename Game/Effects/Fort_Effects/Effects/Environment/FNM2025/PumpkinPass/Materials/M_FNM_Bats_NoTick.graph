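# Masked, unlit, two-sided material for FNM2025 PumpkinPass that animates a swarm of bat sprites in the shader.
# World position offset combines sprite billboarding, time-scrolled curl-noise wander, rotation around the object pivot and distance-based scaling.
# Vertex alpha varies size, speeds and 6x6 flipbook phase per sprite, and feeds a ReduceSpawnCount opacity cull.
# Fractional object position supplies per-component randomness, and the flipbook colour passes through desaturation and shared emissive adjustments.
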Material /Game/Effects/Fort_Effects/Effects/Environment/FNM2025/PumpkinPass/Materials/M_FNM_Bats_NoTick
graph {
 properties {bAutomaticallySetUsageInEditor=false, blendMode=BLEND_Masked, shadingModel=MSM_Unlit, twoSided=true, usedShadingModels=Unlit}
 nodes [
  n1 = WorldPosition()
  n2 = LinearInterpolate(
   A=ScalarParameter() {defaultValue=4, group=WPO-WorldNoise, parameterName=SpriteNoiseSpeed_Min}
   B=ScalarParameter() {defaultValue=7, group=WPO-WorldNoise, parameterName=SpriteNoiseSpeed_Max}
   Alpha=VertexColor().A
  )
  n3 = NamedRerouteUsage()
  n4 = Add(A=WorldPosition().XYZ, B=Multiply(A=Time(), B=Add(A=@n2, B=@n3)))
  n5 = ScalarParameter() {defaultValue=0.001, group=WPO-WorldNoise, parameterName=NoiseTile}
  n6 = Clamp(None=Add(A=Constant() {r=0.1}, B=NamedRerouteUsage())) {maxDefault=0.2, minDefault=0.1}
  n7 = Multiply(
   A=TextureSample(UVs=Multiply(A=@n4, B=@n5)) {samplerType=SAMPLERTYPE_LinearColor, texture=/Game/Effects/FogVolume/VT_CurlNoise_LOW}.RGB
   B=LinearInterpolate(
    B=ScalarParameter() {defaultValue=150, group=WPO-WorldNoise, parameterName=NoiseSineIntensity}
    Alpha=MaterialFunctionCall(Speed=@n6) {materialFunction=&DebugTimeSine}.Result
   ) {constA=1}
  )
  n8 = LinearInterpolate(
   A=ScalarParameter() {defaultValue=160, parameterName=SpriteSize_Min}
   B=ScalarParameter() {defaultValue=180, parameterName=SpriteSize_Max}
   Alpha=VertexColor().A
  ) {constA=160}
  n9 = ScalarParameter() {defaultValue=8000, group=Distance, parameterName=SpriteScale_SmoothScaling}
  n10 = Multiply(
   A=MaterialFunctionCall(
    Fade_Start_Distance=ScalarParameter() {defaultValue=8000, group=Distance, parameterName="Fade_Start_Distance (S)"}
    Fade_End_Distance=ScalarParameter() {defaultValue=6000, group=Distance, parameterName="Fade_End_Distance (S)"}
   ) {materialFunction=&MF_Distance_SyncedFade_FX}.Result
   B=LinearInterpolate(
    A=ScalarParameter() {defaultValue=1, group=Distance, parameterName=SpriteScaleNear}
    B=ScalarParameter() {defaultValue=2, group=Distance, parameterName=SpriteScaleFar}
    Alpha=MaterialFunctionCall(Distance=@n9) {materialFunction=&Near_Camera_Fade}.Result
   )
  )
  n11 = Time()
  n12 = LinearInterpolate(
   A=ScalarParameter() {defaultValue=0.05, group=WPO-Rotation, parameterName=RotationSpeed_Min}
   B=ScalarParameter() {defaultValue=0.08, group=WPO-Rotation, parameterName=RotationSpeed_Max}
   Alpha=VertexColor().A
  )
  Add(
   A=MaterialFunctionCall!(Center Location=Add(A=@n1.XYZ, B=@n7), 0-1 UVs, XY scale=Multiply(A=@n8, B=@n10), OverridePivot, "Normals (Optional)") {materialFunction=&Sprite}.World position offset
   B=RotateAboutAxis(
    NormalizedRotationAxis=VectorParameter() {group=WPO-Rotation, parameterName=RotationAxis, defaultValue=<rgba 0 0 1 1>}.RGB
    RotationAngle=Multiply(A=@n11, B=@n12)
    PivotPoint=ObjectPositionWS()
    Position=WorldPosition().XYZ
   )
  )
  n13 = Multiply(A=VertexColor().A, B=ScalarParameter() {parameterName=ReduceSpawnCount})
  n14 = NamedRerouteUsage()
  Add(A=@n13, B=Ceil(None=@n14))
  n15 = Time()
  n16 = Multiply(A=@n15, B=ScalarParameter() {defaultValue=1, parameterName=FlipBookAnimationMult})
  n17 = VertexColor()
  n18 = Constant() {r=6}
  n19 = TextureSample(UVs=MaterialFunctionCall("Animation  Phase (0-1)"=LinearInterpolate(A=@n15, B=@n16, Alpha=@n17.A), Number of Rows=@n18, Number of Columns=@n18) {materialFunction=&FlipBook}.UVs) {texture=/Game/Effects/Fort_Effects/Effects/Environment/FNM2025/PumpkinPass/Textures/T_FNM_Bats_NoTick1}
  MaterialFunctionCall(Emissive=Desaturation(None=@n19.RGB, Fraction=ScalarParameter() {parameterName=DesaturateTexture})) {materialFunction=&MF_FX_EmissiveAdjustments}
  Multiply(A=@n19.A, B=LinearInterpolate(Alpha=Step(Y=@n13) {constY=1}))
  n20 = ObjectPositionWS()
  NamedRerouteDeclaration(None=ComponentMask(None=Frac(None=@n20)) {g=true}) {name=PerComponentRandom, nodeColor=<rgba 0 0.2969 1 1>}
  Round(None=@n14)
 ]
}
