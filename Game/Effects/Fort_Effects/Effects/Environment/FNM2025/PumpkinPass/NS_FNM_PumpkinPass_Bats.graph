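# Niagara environment effect for FNM2025 PumpkinPass that swirls mesh bats around the system origin, spawning them in rings and culling by distance.
# The GPU emitter Bats001 combines vortex pull, curl noise, camera repulsion via negative point attraction, distance-field avoidance and kill-on-collision.
# A simpler, slower CPU emitter runs only the vortex; the GPU emitter Bats is disabled.
# Curves over lifetime and distance scale the meshes, and dynamic parameters feed their material.
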NiagaraSystem /Game/Effects/Fort_Effects/Effects/Environment/FNM2025/PumpkinPass/NS_FNM_PumpkinPass_Bats
dependencies {
 usedDataInterfaces [&NiagaraDataInterfaceCurve, &NiagaraDataInterfaceCamera]
 usedDynamicInputs []
 usedModules [&SystemState]
}
emitters {
 Bats {
  bEnabled=false
  simTarget=GPUComputeSim
  emitterSpawnScript [/SetVariables(Emitter.RadiusMult=1) {isSetParameters=true, module=SetVariables_#1}]
  emitterUpdateScript [
   &EmitterState(
    Enable Distance Culling=true
    Life Cycle Mode=::NewEnumerator1
    Loop Behavior=::NewEnumerator1
    Loop Duration Mode=::NewEnumerator1
    MaxDistance=10000
    Min Distance Response=::SleepAndLetParticlesFinish
    Module.bApplySpawnCountScaleByDistance=true
    Module.bUseMaxDistance=true
    Module.bUseMinDistance=true
    Scalability Mode=::NewEnumerator1
    Scale Spawn Count=true
    Spawn Count Scale By Distance=DataInterface() {propertyValues {Curve {keys [<at 0 1>,<at 0.95 1>,<at 1 0>]}}}
    UseLoopDelay=false
   )
   &SpawnRate(SpawnRate=&RandomRangeFloat(Minimum=5, Maximum=9), Use Spawn Probability=false) {module=SpawnRate001}
   /SetVariables(Emitter.CameraPosition:Vector3f=&ReturnCameraPosition(Coordinate Space:ENiagaraCoordinateSpace=::Simulation, Camera Query:NiagaraDataInterfaceCamera=DataInterface())) {isSetParameters=true, module=SetVariables_#2}
  ]
  particleSpawnScript [
   &PartitionParticles()
   &InitializeParticle(
    Color=<rgba 3 3 3 1>
    Lifetime=&RandomRangeFloat(Minimum=8, Maximum=10)
    Mesh Scale Mode=::NewEnumerator3
    Mesh Uniform Scale=&Multiply_Float(A=&RandomRangeFloat(Minimum=0.15, Maximum=0.3), B=2)
    UsePositionOffset=false
    Write Material Random=true
   )
   &ShapeLocation(
    Enable Torus Knot Radius=false
    Handle Radius:NiagaraFloat=&Multiply_Float(A=300, B=$Emitter.RadiusMult)
    Large Radius=&Multiply_Float(A=&RandomRangeFloat(Minimum=1500, Maximum=3000), B=$Emitter.RadiusMult)
    MaskSpawnGroup=false
    Offset:NiagaraPosition=$Emitter.PositionOffset
    Offset Mode=::NewEnumerator0
    Override Seed=false
    Shape Primitive=::NewEnumerator3
    Torus Mode:ENiagaraShapeTorusMode=::NewEnumerator1
    UseOwnerScale=true
   )
  ]
  particleUpdateScript [
   &ParticleState()
   &PointAttractionForce(
    Attraction Radius=2500
    Attraction Strength=-500
    Attractor Position:Vector3f=$Emitter.CameraPosition
    Kill Within Radius=false
    Module.Use Falloff=true
    UseAttractorOffset=false
   ) {module=PointAttractionForce001}
   &VortexForce(
    Origin Pull Amount:NiagaraFloat=&RandomRangeFloat(Minimum=50, Maximum=120)
    Origin Pull Only On Vortex Axis=false
    UseOriginOffset=true
    Vortex Axis=<0 0 4>
    Vortex Force Amount=100
    Vortex Origin:NiagaraPosition=&AddVectorToPosition(Position=$Engine.Emitter.SimulationPosition, Vector=Linked() {linkedVariable {name=Emitter.PositionOffset, type=NiagaraPosition}})
    Vortex Origin Offset:Vector3f=&RandomRangeVector(Minimum=<0 0 222>, Maximum=<0 0 -222>)
   )
   /Niagara/Modules/Math/AvoidDistanceFieldSurfaces_GPU(Avoid Nearest Surface=true, Avoid Oncoming Surfaces=true, Nearest Surface Avoidance Strength=800, Obstactle Avoidance Strength=800)
   &Collision(
    Age Colliding Particles=false
    GPU Collision Type=::NewEnumerator2
    KillOnCollision=false
    Module.Control Rest State Via Surface Normal=false
    Module.EnableMaxCollisionCount=false
    Module.Randomize Collision Normal Vector=false
    Radius Calculation Type=::NewEnumerator1
    Restitution=1
   )
   &CurlNoiseForce(
    Bias Noise=false
    Noise Frequency=&UniformAOrBFloat(
     A=0
     B=&RandomRangeFloat(Minimum=19, Maximum=5)
     Distribution Weight=0.15
     Evaluation Type=::NewEnumerator0
     Recalculate Random Each Loop=true
     Override Seed=false
    )
    Noise Strength=&UniformAOrBFloat(
     A=0
     B=&RandomRangeFloat(Minimum=555, Maximum=955)
     Distribution Weight=0.15
     Evaluation Type=::NewEnumerator0
     Recalculate Random Each Loop=true
     Override Seed=false
    )
    Pan Noise=true
    Pan Noise Field=<1 0 0>
   ) {module=CurlNoiseForce001}
   &SolveForcesAndVelocity(Clamp Velocity=true, Limit Acceleration=false, Speed Limit=900)
   /SetVariables(Particles.New Vector 3f:Vector3f=&Multiply_VectorByFloat(
     Vector=$Particles.Scale
     Float=&FloatFromCurve(FloatCurve=DataInterface() {propertyValues {Curve {keys [
          {interpMode=RCIM_Cubic, tangentMode=RCTM_User, time=0, value=0}
          {interpMode=RCIM_Cubic, tangentMode=RCTM_User, time=0.1, value=1}
          {interpMode=RCIM_Cubic, tangentMode=RCTM_User, time=0.8, value=1}
          {tangentMode=RCTM_User, time=1, value=0}
         ]}}})
    )) {isSetParameters=true, module=SetVariables_#3}
   &DynamicMaterialParameters(
    Index 0 Param 1=0.25
    Index 0 Param 2=&RandomRangeFloat(Minimum=2.8, Maximum=3.5)
    Index 0 Param 3=0.25
    Index 0 Param 4=0.1
    Index 1 Param 1=1
    Index 1 Param 2=&RandomRangeFloat(Minimum=25, Maximum=55)
    Index 1 Param 3=2
    Index 1 Param 4=0.25
    Index 2 Param 1=1
    Index 2 Param 2=1
    Index 2 Param 3=2
    Index 2 Param 4=0.25
    Module.Index0Param0Write=true
    Module.Index0Param1Write=true
    Module.Index0Param2Write=true
    Module.Index0Param3Write=true
    Module.Index1Param0Write=true
    Module.Index1Param1Write=true
    Module.Index1Param2Write=true
    Module.Index1Param3Write=true
    Module.Index2Param0Write=true
    Module.Index2Param1Write=true
    Module.Index2Param2Write=true
    Module.Index2Param3Write=true
    Write Parameter Index 1=true
    Write Parameter Index 2=true
   ) {module=DynamicMaterialParameters001}
  ]
  rendererClasses [&NiagaraMeshRendererProperties]
  renderers [{rendererClass=&NiagaraMeshRendererProperties, rendererIndex=0}]
 }
 Bats001 {
  simTarget=GPUComputeSim
  emitterSpawnScript [/SetVariables(Emitter.RadiusMult=1) {isSetParameters=true, module=SetVariables_#1}]
  emitterUpdateScript [
   &EmitterState(
    Enable Distance Culling=true
    Life Cycle Mode=::NewEnumerator1
    Loop Behavior=::NewEnumerator1
    Loop Duration Mode=::NewEnumerator1
    MaxDistance=10000
    Min Distance Response=::SleepAndLetParticlesFinish
    Module.bApplySpawnCountScaleByDistance=true
    Module.bUseMaxDistance=true
    Module.bUseMinDistance=true
    Scalability Mode=::NewEnumerator1
    Scale Spawn Count=true
    Spawn Count Scale By Distance=DataInterface() {propertyValues {Curve {keys [<at 0 1>,<at 0.95 1>,<at 1 0>]}}}
    UseLoopDelay=false
   )
   &SpawnRate(SpawnRate=&RandomRangeFloat(Minimum=3, Maximum=6), Use Spawn Probability=false) {module=SpawnRate001}
   /SetVariables(Emitter.CameraPosition:Vector3f=&ReturnCameraPosition(Coordinate Space:ENiagaraCoordinateSpace=::Simulation, Camera Query:NiagaraDataInterfaceCamera=DataInterface())) {isSetParameters=true, module=SetVariables_#2}
  ]
  particleSpawnScript [
   &PartitionParticles()
   &InitializeParticle(
    Color=<rgba 3 3 3 1>
    Lifetime=&RandomRangeFloat(Minimum=8, Maximum=10)
    Mesh Scale Mode=::NewEnumerator3
    Mesh Uniform Scale=&Multiply_Float(A=&RandomRangeFloat(Minimum=0.1, Maximum=0.25), B=2)
    UsePositionOffset=false
    Write Material Random=true
   )
   &ShapeLocation(
    Enable Torus Knot Radius=false
    Handle Radius:NiagaraFloat=&Multiply_Float(A=300, B=$Emitter.RadiusMult)
    Large Radius=1000
    MaskSpawnGroup=false
    Offset:NiagaraPosition=$Emitter.PositionOffset
    Offset Mode=::NewEnumerator0
    Override Seed=false
    Shape Primitive=::NewEnumerator3
    Torus Mode:ENiagaraShapeTorusMode=::NewEnumerator1
    UseOwnerScale=true
   )
  ]
  particleUpdateScript [
   &ParticleState()
   &ScaleMeshSize(Scale Factor=&VectorFromFloat(Value=&FloatFromCurve(FloatCurve=DataInterface() {propertyValues {Curve {
         defaultValue=1
         keys [
          {interpMode=RCIM_Cubic, tangentMode=RCTM_User, time=0, value=0}
          {arriveTangent=2, interpMode=RCIM_Cubic, leaveTangent=0.0924, tangentMode=RCTM_Break, time=0.1, value=1}
          {interpMode=RCIM_Cubic, tangentMode=RCTM_User, time=0.5, value=1}
          {tangentMode=RCTM_User, time=1, value=0}
         ]
        }}})))
   &PointAttractionForce(
    Attraction Radius=1000
    Attraction Strength=-500
    Attractor Position:Vector3f=$Emitter.CameraPosition
    Kill Within Radius=false
    Module.Use Falloff=true
    UseAttractorOffset=false
   ) {module=PointAttractionForce001}
   &VortexForce(
    Origin Pull Amount:NiagaraFloat=&RandomRangeFloat(Minimum=50, Maximum=120)
    Origin Pull Only On Vortex Axis=false
    UseOriginOffset=true
    Vortex Axis=<0 0 4>
    Vortex Force Amount=100
    Vortex Origin:NiagaraPosition=&AddVectorToPosition(Position=$Engine.Emitter.SimulationPosition, Vector=Linked() {linkedVariable {name=Emitter.PositionOffset, type=NiagaraPosition}})
    Vortex Origin Offset:Vector3f=&RandomRangeVector(Minimum=<0 0 222>, Maximum=<0 0 -222>)
   )
   &AvoidDistanceFieldSurfaces_GPU(Avoid Nearest Surface=true, Avoid Oncoming Surfaces=true, Nearest Surface Avoidance Strength=800, Obstactle Avoidance Strength=800)
   &Collision(
    Age Colliding Particles=false
    GPU Collision Type=::NewEnumerator2
    Kill On Collision=9
    KillOnCollision=true
    Module.Control Rest State Via Surface Normal=false
    Module.EnableMaxCollisionCount=false
    Module.Randomize Collision Normal Vector=false
    Particle Radius Scale=5
    Radius Calculation Type=::NewEnumerator1
    Restitution=1
   )
   &CurlNoiseForce(
    Bias Noise=false
    Noise Frequency=&UniformAOrBFloat(
     A=0
     B=&RandomRangeFloat(Minimum=19, Maximum=5)
     Distribution Weight=0.15
     Evaluation Type=::NewEnumerator0
     Recalculate Random Each Loop=true
     Override Seed=false
    )
    Noise Strength=&UniformAOrBFloat(
     A=0
     B=&RandomRangeFloat(Minimum=555, Maximum=955)
     Distribution Weight=0.15
     Evaluation Type=::NewEnumerator0
     Recalculate Random Each Loop=true
     Override Seed=false
    )
    Pan Noise=true
    Pan Noise Field=<1 0 0>
   ) {module=CurlNoiseForce001}
   &SolveForcesAndVelocity(Clamp Velocity=true, Limit Acceleration=false, Speed Limit=500)
   /SetVariables(Particles.New Vector 3f:Vector3f=&Multiply_VectorByFloat(
     Vector=$Particles.Scale
     Float=&Multiply_Float(
      A=&FloatFromCurve(FloatCurve=DataInterface() {propertyValues {Curve {keys [
           {interpMode=RCIM_Cubic, tangentMode=RCTM_User, time=0, value=0}
           {interpMode=RCIM_Cubic, tangentMode=RCTM_User, time=0.1, value=1}
           {interpMode=RCIM_Cubic, tangentMode=RCTM_User, time=0.8, value=1}
           {tangentMode=RCTM_User, time=1, value=0}
          ]}}})
      B=&FloatFromCurve(
       FloatCurve=DataInterface() {propertyValues {Curve {
          defaultValue=1
          keys [
           {interpMode=RCIM_Cubic, tangentMode=RCTM_User, time=0, value=0}
           {arriveTangent=11.0185, interpMode=RCIM_Cubic, tangentMode=RCTM_Break, time=0.1, value=1}
          ]
         }}}
       CurveIndex=$Emitter.Scalability.DistanceFraction
      )
     )
    )) {isSetParameters=true, module=SetVariables_#3}
   &DynamicMaterialParameters(
    Index 0 Param 1=0.25
    Index 0 Param 2=&RandomRangeFloat(Minimum=2.8, Maximum=3.5)
    Index 0 Param 3=0.25
    Index 0 Param 4=0.1
    Index 1 Param 1=1
    Index 1 Param 2=&RandomRangeFloat(Minimum=25, Maximum=55)
    Index 1 Param 3=2
    Index 1 Param 4=0.25
    Index 2 Param 1=1
    Index 2 Param 2=1
    Index 2 Param 3=2
    Index 2 Param 4=0.25
    Module.Index0Param0Write=true
    Module.Index0Param1Write=true
    Module.Index0Param2Write=true
    Module.Index0Param3Write=true
    Module.Index1Param0Write=true
    Module.Index1Param1Write=true
    Module.Index1Param2Write=true
    Module.Index1Param3Write=true
    Module.Index2Param0Write=true
    Module.Index2Param1Write=true
    Module.Index2Param2Write=true
    Module.Index2Param3Write=true
    Write Parameter Index 1=true
    Write Parameter Index 2=true
   ) {module=DynamicMaterialParameters001}
  ]
  rendererClasses [&NiagaraMeshRendererProperties]
  renderers [{rendererClass=&NiagaraMeshRendererProperties, rendererIndex=0}]
 }
 Bats_CPU {
  emitterUpdateScript [
   &EmitterState(
    Enable Distance Culling=true
    Life Cycle Mode=::NewEnumerator1
    Loop Behavior=::NewEnumerator1
    Loop Duration Mode=::NewEnumerator1
    MaxDistance=10000
    Min Distance Response=::SleepAndLetParticlesFinish
    Module.bApplySpawnCountScaleByDistance=true
    Module.bUseMaxDistance=true
    Module.bUseMinDistance=true
    Scalability Mode=::NewEnumerator1
    Scale Spawn Count=true
    Spawn Count Scale By Distance=DataInterface() {propertyValues {Curve {keys [<at 0 1>,<at 0.95 1>,<at 1 0>]}}}
    UseLoopDelay=false
   )
   &SpawnRate(SpawnRate=2, Use Spawn Probability=false) {module=SpawnRate001}
  ]
  particleSpawnScript [
   &InitializeParticle(
    Color=<rgba 3 3 3 1>
    Lifetime=&RandomRangeFloat(Minimum=8, Maximum=10)
    Mesh Scale Mode=::NewEnumerator3
    Mesh Uniform Scale=&Multiply_Float(A=&RandomRangeFloat(Minimum=0.1, Maximum=0.25), B=2)
    UsePositionOffset=false
    Write Material Random=true
   )
   &ShapeLocation(
    Enable Torus Knot Radius=false
    Handle Radius:NiagaraFloat=&Multiply_Float(A=300, B=$Emitter.RadiusMult)
    Large Radius=600
    MaskSpawnGroup=false
    Offset=<0 0 -444>
    Offset Mode=::NewEnumerator0
    Override Seed=false
    Shape Primitive=::NewEnumerator3
    Torus Mode:ENiagaraShapeTorusMode=::NewEnumerator1
    UseOwnerScale=true
   )
  ]
  particleUpdateScript [
   &ParticleState()
   &VortexForce(
    Origin Pull Amount:NiagaraFloat=&RandomRangeFloat(Minimum=50, Maximum=120)
    Origin Pull Only On Vortex Axis=false
    UseOriginOffset=true
    Vortex Axis=<0 0 4>
    Vortex Force Amount=100
    Vortex Origin:NiagaraPosition=&AddVectorToPosition(Position=$Engine.Emitter.SimulationPosition, Vector=Linked() {linkedVariable {name=Emitter.PositionOffset, type=NiagaraPosition}})
    Vortex Origin Offset:Vector3f=&RandomRangeVector(Minimum=<0 0 99>, Maximum=<0 0 0>)
   )
   &SolveForcesAndVelocity(Clamp Velocity=true, Limit Acceleration=false, Speed Limit=200)
   /SetVariables(Particles.New Vector 3f:Vector3f=&Multiply_VectorByFloat(
     Vector=$Particles.Scale
     Float=&Multiply_Float(
      A=&FloatFromCurve(FloatCurve=DataInterface() {propertyValues {Curve {keys [
           {interpMode=RCIM_Cubic, tangentMode=RCTM_User, time=0, value=0}
           {interpMode=RCIM_Cubic, tangentMode=RCTM_User, time=0.1, value=1}
           {interpMode=RCIM_Cubic, tangentMode=RCTM_User, time=0.8, value=1}
           {tangentMode=RCTM_User, time=1, value=0}
          ]}}})
      B=&FloatFromCurve(
       FloatCurve=DataInterface() {propertyValues {Curve {
          defaultValue=1
          keys [
           {interpMode=RCIM_Cubic, tangentMode=RCTM_User, time=0, value=0}
           {arriveTangent=11.0185, interpMode=RCIM_Cubic, tangentMode=RCTM_Break, time=0.13, value=1}
          ]
         }}}
       CurveIndex=$Emitter.Scalability.DistanceFraction
      )
     )
    )) {isSetParameters=true, module=SetVariables_#3}
   &DynamicMaterialParameters(
    Index 0 Param 1=0.25
    Index 0 Param 2=&RandomRangeFloat(Minimum=2.8, Maximum=3.5)
    Index 0 Param 3=0.25
    Index 0 Param 4=0.1
    Index 1 Param 1=1
    Index 1 Param 2=&RandomRangeFloat(Minimum=25, Maximum=55)
    Index 1 Param 3=2
    Index 1 Param 4=0.25
    Index 2 Param 1=1
    Index 2 Param 2=1
    Index 2 Param 3=2
    Index 2 Param 4=0.25
    Module.Index0Param0Write=true
    Module.Index0Param1Write=true
    Module.Index0Param2Write=true
    Module.Index0Param3Write=true
    Module.Index1Param0Write=true
    Module.Index1Param1Write=true
    Module.Index1Param2Write=true
    Module.Index1Param3Write=true
    Module.Index2Param0Write=true
    Module.Index2Param1Write=true
    Module.Index2Param2Write=true
    Module.Index2Param3Write=true
    Write Parameter Index 1=true
    Write Parameter Index 2=true
   ) {module=DynamicMaterialParameters001}
  ]
  rendererClasses [&NiagaraMeshRendererProperties]
  renderers [{rendererClass=&NiagaraMeshRendererProperties, rendererIndex=0}]
 }
}
summary {
 systemName=NS_FNM_PumpkinPass_Bats
 emitters [
  {bEnabled=false, emitterName=Bats, simTarget=GPUComputeSim, rendererClasses [&NiagaraMeshRendererProperties]}
  {emitterName=Bats001, simTarget=GPUComputeSim, rendererClasses [&NiagaraMeshRendererProperties]}
  {emitterName=Bats_CPU, rendererClasses [&NiagaraMeshRendererProperties]}
 ]
 userVariables [User.Radius Mult {type=NiagaraFloat, defaultValue=1, description=""}]
}
