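# GPU Niagara effect for FNM2025 PumpkinPass: crows perch on scenery, then fly off as the player nears.
# Spawning snaps crows to distance-field surfaces and scene depth, filtering by surface normal, valid depth and GBuffer hue.
# A latched player-distance check switches visibility tag, starts aging and launches flight with surface avoidance, collision kills and end-of-life shrinking.
# Only Crows_SceneDepth001 is enabled; a hard-coded world-space kill volume ties it to one location.
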NiagaraSystem /Game/Effects/Fort_Effects/Effects/Environment/FNM2025/PumpkinPass/NS_FNM_PumpkinPass_Crows_IdleToFlight
dependencies {
 usedDataInterfaces [&NiagaraDataInterfaceCurve, &NiagaraDataInterfaceCollisionQuery, &NiagaraDataInterfaceGBuffer]
 usedDynamicInputs []
 usedModules [&SystemState]
}
emitters {
 Crows_SceneDepth {
  bEnabled=false
  simTarget=GPUComputeSim
  emitterUpdateScript [
   &EmitterState(Inactive Response=::NewEnumerator2, Life Cycle Mode=::NewEnumerator1, Loop Behavior=::NewEnumerator1, UseLoopDelay=false)
   &SpawnRate(
    SpawnRate=&FloatFromCurve(FloatCurve=DataInterface() {propertyValues {Curve {keys [<at 0 1>,<at 5 0>]}}}, CurveIndex=$Emitter.Age, Scale Curve=4444)
    Use Spawn Probability=false
   ) {enabled=false}
   &SpawnRate(
    SpawnRate=&FloatFromCurve(
     FloatCurve=DataInterface() {propertyValues {Curve {keys [<at 0 3>,<at 1 0.7982>,<at 5 0>]}}}
     CurveIndex=$Emitter.Age
     Scale Curve=&RandomRangeFloat(Minimum=9000, Maximum=9000)
    )
    Use Spawn Probability=false
   ) {module=SpawnRate001}
  ]
  particleSpawnScript [
   &PartitionParticles()
   &InitializeParticle(
    Color Maximum=<rgba 0.387 0.387 0.387 1>
    Color Minimum=<rgba 1 1 1 5>
    Color Mode=::NewEnumerator2
    Lifetime Max=8
    Lifetime Min=7
    Lifetime Mode=::NewEnumerator1
    Mass Mode=::NewEnumerator1
    Mesh Scale=&Multiply_Vector(
     A=&RandomRangeVector(Minimum=<1 1.1 1.1>, Maximum=<1 1 1>)
     B=&VectorFromFloat(Value=&RandomRangeFloat(Minimum=0.7, Maximum=1))
    )
    Mesh Scale Mode=::NewEnumerator4
    Sprite Rotation Mode=::NewEnumerator2
    Sprite Size Mode=::NewEnumerator3
    Uniform Sprite Size=15
    UsePositionOffset=false
    Write Material Random=true
   ) {module=InitializeParticle001}
   &ShapeLocation(MaskSpawnGroup=false, Non Uniform Scale=<2 2 0.75>, Override Seed=false, Sphere Radius=2000, UseOwnerScale=true)
   &AddVelocity(Velocity=&RandomRangeVector(Minimum=<1 -1 0>, Maximum=<0 1 0>)) {module=AddVelocity001}
   /Niagara/Modules/Math/MoveToNearestDistanceFieldSurface_GPU(Offset Intrinsic Position By Radius=true, bKillByDistance=false)
   /SetVariables(Particles.SurfaceNormal:Vector3f=&Multiply_VectorByFloat(
     Vector=&NormalizeVector(
      Vector To Normalize=$Output.MoveToNearestDistanceFieldSurface_GPU.GlobalDistanceFieldAveragedNormal
      Safe Normalize=true
      Safe Fallback Threshold=0
      Safe Fallback Vector=<0 0 0.001>
     )
     Float=-1
    )) {isSetParameters=true, module=SetVariables_#1}
   /Niagara/Modules/Math/PlaceParticlesOnDepthBuffer_GPU(Max Z Depth=9000, Module.Use Max Z Depth=true)
   &KillParticles(Kill Particles=&SetBoolByFloatComparison(A=&DotProduct(Vector A=$Particles.SurfaceNormal, Vector B=<0 0 1>), B=0.8, Comparison Type=::NewEnumerator6))
   &self:ScratchModule_02(Angle=55, BlendDirection=1, Normal:Vector3f=<0 0 1>, PickCeilingOrFloor=true) {module=ScratchModule_02}
   /Game/Effects/Niagara/Modules/Spawn/FN_Spawn_GBuffer_Color_WithinRangeV2(
    HueMax:NiagaraFloat=&Lerp_Float(A=360, B=164, Alpha=&MakeFloatFromInt(Int=$Particles.PartitionParticles.Partition))
    HueMin:NiagaraFloat=&Lerp_Float(A=355, B=152, Alpha=&MakeFloatFromInt(Int=$Particles.PartitionParticles.Partition))
   )
   /SetVariables() {isSetParameters=true, module=SetVariables_#2}
   &Color(Color=&RandomRangeLinearColor(Minimum=<rgba 0.0144 0.0666 0.0513 1>, Maximum=<rgba 0.0242 0.1195 0.0762 1>)) {enabled=false, module=Color001}
   /SetVariables() {isSetParameters=true, module=SetVariables_#3}
   /SetVariables(
    Particles.DebugScale:Vector2f=&Vector2DFromFloat(Value=55)
    Particles.GnomonScale:Vector3f=&VectorFromFloat(Value=9)
   ) {isSetParameters=true, module=SetVariables_#4}
  ]
  particleUpdateScript [
   /SetVariables(Particles.Is Active:NiagaraBool=&Boolean_OrOperation(
     Bool1=$Particles.Is Active
     Bool2=&SetBoolByFloatComparison(
      A=&RandomRangeFloat(Minimum=2500, Maximum=6000)
      B=&DistanceBetweenPositions(A=$Particles.Position, B=&MakeVectorFromLinearColorRGB(Color=$NPC.FortniteNPC.PlayerPosition))
      Comparison Type=::NewEnumerator2
     )
    )) {isSetParameters=true, module=SetVariables_#5}
   /SetVariables(Particles.VisibilityTag=&MakeCustomIntFromBool(Bool=$Particles.Is Active, True INT=1, False INT=0)) {isSetParameters=true, module=SetVariables_#6}
   &ParticleState(DeltaTime:NiagaraFloat=&MakeCustomFloatFromBool(Bool=$Particles.Is Active, True Float=$Engine.DeltaTime, False Float=0)) {module=ParticleState001}
   /SetVariables(Particles.Velocity=&MakeCustomVectorFromBool(
     Bool=$Particles.Is Active
     True Vector=&Multiply_VectorByFloat(
      Vector=&RandomRangeVector(Minimum=<-1250 0 1200>, Maximum=<-1000 0 700>)
      Float=&FloatFromCurve(FloatCurve=DataInterface() {propertyValues {Curve {keys [
           {interpMode=RCIM_Cubic, tangentMode=RCTM_User, time=0, value=0.2}
           {arriveTangent=0.3, interpMode=RCIM_Cubic, leaveTangent=0.3, time=0.2, value=0.4}
           {interpMode=RCIM_Cubic, tangentMode=RCTM_User, time=1, value=0.5}
          ]}}})
     )
     False Vector=<0 0 0>
    )) {isSetParameters=true, module=SetVariables_#7}
   &UpdateMeshOrientation(
    Rotation Rate=&MakeCustomFloatFromBool(Bool=$Particles.Is Active, True Float=0, False Float=1)
    Rotation Vector=&MakeVector(
     X=&Cosine(
      Normalized Angle=$Emitter.Age
      Period=&RandomRangeFloat(Minimum=1, Maximum=0.3)
      Scale=&UniformAOrBFloat(
       A=0
       B=&RandomRangeFloat(Minimum=0.7, Maximum=0.5)
       Distribution Weight=&RandomRangeFloat(Minimum=0.5, Maximum=0.7)
       Evaluation Type=::NewEnumerator0
       Recalculate Random Each Loop=true
       Override Seed=false
      )
      Bias=0
     )
     Y=0
     Z=0
    )
   ) {enabled=false}
   /Niagara/Modules/Math/AvoidDistanceFieldSurfaces_GPU(
    Active:NiagaraBool=$Particles.Is Active
    Avoid Nearest Surface=true
    Avoid Oncoming Surfaces=true
    Nearest Surface Avoidance Strength=25000
    Obstactle Avoidance Strength=25000
   ) {module=AvoidDistanceFieldSurfaces_GPU001}
   &Collision(
    Age Colliding Particles=false
    Enable Rest State=false
    GPU Collision Type=::NewEnumerator2
    Kill On Collision=15
    KillOnCollision=true
    Module.EnableMaxCollisionCount=false
    Module.Randomize Collision Normal Vector=false
    Particle Radius Scale=9
    Radius Calculation Type=::NewEnumerator1
    Restitution=0
   ) {module=Collision001}
   &SolveForcesAndVelocity(Clamp Velocity=true, Limit Acceleration=true)
   &ScaleMeshSize(Scale Factor=&VectorFromFloat(Value=&FloatFromCurve(FloatCurve=DataInterface() {propertyValues {Curve {keys [
          {tangentMode=RCTM_User, time=0, value=1}
          {arriveTangent=-0.0405, interpMode=RCIM_Cubic, leaveTangent=-0.0405, tangentMode=RCTM_User, time=0.85, value=1}
          {arriveTangent=-8.0089, interpMode=RCIM_Cubic, leaveTangent=-8.0089, tangentMode=RCTM_User, time=1, value=0}
         ]}}})))
   &KillParticlesInVolume(
    Box Size:Vector3f=<3000 15555 3000>
    Kill Shape=::NewEnumerator1
    Volume Origin:Vector3f=&AddVectorToPosition(Position=<88272 9968 4316>, Vector=<0 0 0>)
   ) {module=KillParticlesInVolume002}
   &DynamicMaterialParameters(
    Index 0 Param 2=&RandomRangeFloat(Minimum=2.1, Maximum=2.7)
    Module.Index0Float0Write=false
    Module.Index0Float1Write=true
    Module.Index0Float2Write=false
    Module.Index0Float3Write=false
   )
  ]
  rendererClasses [&NiagaraMeshRendererProperties, &NiagaraSpriteRendererProperties]
  renderers [
   {rendererClass=&NiagaraMeshRendererProperties, rendererIndex=0}
   {rendererClass=&NiagaraMeshRendererProperties, rendererIndex=1}
   {rendererClass=&NiagaraMeshRendererProperties, rendererIndex=2}
   {rendererClass=&NiagaraSpriteRendererProperties, rendererIndex=3}
  ]
 }
 Crows_SceneDepth001 {
  simTarget=GPUComputeSim
  emitterUpdateScript [
   &EmitterState(Inactive Response=::NewEnumerator2, Life Cycle Mode=::NewEnumerator1, Loop Behavior=::NewEnumerator1, UseLoopDelay=false)
   &SpawnRate(
    SpawnRate=&FloatFromCurve(FloatCurve=DataInterface() {propertyValues {Curve {keys [<at 0 1>,<at 5 0>]}}}, CurveIndex=$Emitter.Age, Scale Curve=4444)
    Use Spawn Probability=false
   ) {enabled=false}
   &SpawnRate(
    SpawnRate=&FloatFromCurve(
     FloatCurve=DataInterface() {propertyValues {Curve {keys [<at 0 3>,<at 1 0.7982>,<at 5 0>]}}}
     CurveIndex=$Emitter.Age
     Scale Curve=&RandomRangeFloat(Minimum=2000, Maximum=2500)
    )
    Use Spawn Probability=false
   ) {module=SpawnRate001}
  ]
  particleSpawnScript [
   &PartitionParticles()
   &InitializeParticle(
    Color Maximum=<rgba 0.387 0.387 0.387 1>
    Color Minimum=<rgba 1 1 1 5>
    Color Mode=::NewEnumerator2
    Lifetime Max=8
    Lifetime Min=7
    Lifetime Mode=::NewEnumerator1
    Mass Mode=::NewEnumerator1
    Mesh Scale=&Multiply_Vector(
     A=&RandomRangeVector(Minimum=<1 1.1 1.1>, Maximum=<1 1 1>)
     B=&VectorFromFloat(Value=&RandomRangeFloat(Minimum=0.7, Maximum=1))
    )
    Mesh Scale Mode=::NewEnumerator4
    Sprite Rotation Mode=::NewEnumerator2
    Sprite Size Mode=::NewEnumerator3
    Uniform Sprite Size=15
    UsePositionOffset=false
    Write Material Random=true
   ) {module=InitializeParticle001}
   &ShapeLocation(MaskSpawnGroup=false, Non Uniform Scale=<2 2 0.75>, Override Seed=false, Sphere Radius=2000, UseOwnerScale=true)
   &AddVelocity(Velocity=&RandomRangeVector(Minimum=<1 -1 0>, Maximum=<0 1 0>)) {module=AddVelocity001}
   &MoveToNearestDistanceFieldSurface_GPU(Offset Intrinsic Position By Radius=true, bKillByDistance=false)
   &PlaceParticlesOnDepthBuffer_GPU(Max Z Depth=12000, Module.Use Max Z Depth=true)
   /SetVariables(Particles.SurfaceNormal:Vector3f=&Multiply_VectorByFloat(
     Vector=&NormalizeVector(
      Vector To Normalize=$Output.MoveToNearestDistanceFieldSurface_GPU.GlobalDistanceFieldAveragedNormal
      Safe Normalize=true
      Safe Fallback Threshold=0
      Safe Fallback Vector=<0 0 0.001>
     )
     Float=-1
    )) {isSetParameters=true, module=SetVariables_#1}
   &KillParticles(Kill Particles=&SetBoolByFloatComparison(A=&DotProduct(Vector A=$Particles.SurfaceNormal, Vector B=<0 0 1>), B=0.8, Comparison Type=::NewEnumerator6))
   &self:ScratchModule_02(Angle=55, BlendDirection=1, Normal:Vector3f=<0 0 1>, PickCeilingOrFloor=true) {module=ScratchModule_02}
   &FN_Spawn_GBuffer_Color_WithinRangeV2(
    HueMax:NiagaraFloat=&Lerp_Float(A=360, B=170, Alpha=&MakeFloatFromInt(Int=$Particles.PartitionParticles.Partition))
    HueMin:NiagaraFloat=&Lerp_Float(A=355, B=150, Alpha=&MakeFloatFromInt(Int=$Particles.PartitionParticles.Partition))
   )
   &KillParticles(Kill Particles=&Boolean_NotOperation(Bool=$Output.PlaceParticlesOnDepthBuffer_GPU.ValidSample)) {module=KillParticles001}
   /SetVariables() {isSetParameters=true, module=SetVariables_#2}
   /SetVariables() {isSetParameters=true, module=SetVariables_#3}
   /SetVariables(
    Particles.DebugScale:Vector2f=&Vector2DFromFloat(Value=55)
    Particles.GnomonScale:Vector3f=&VectorFromFloat(Value=9)
   ) {isSetParameters=true, module=SetVariables_#4}
   &Color(Color=&RandomRangeLinearColor(Minimum=<rgba 0.0137 0.0452 0.0395 1>, Maximum=<rgba 0.0232 0.1195 0.0742 1>)) {enabled=false}
   &Color(Color=&RandomRangeLinearColor(Minimum=<rgba 0.1095 0.0513 0.0529 1>, Maximum=<rgba 0.1119 0.0762 0.0545 1>)) {enabled=false, module=Color002}
  ]
  particleUpdateScript [
   /SetVariables(Particles.Is Active:NiagaraBool=&Boolean_OrOperation(
     Bool1=$Particles.Is Active
     Bool2=&SetBoolByFloatComparison(
      A=&RandomRangeFloat(Minimum=2500, Maximum=6000)
      B=&DistanceBetweenPositions(A=$Particles.Position, B=&MakeVectorFromLinearColorRGB(Color=$NPC.FortniteNPC.PlayerPosition))
      Comparison Type=::NewEnumerator2
     )
    )) {isSetParameters=true, module=SetVariables_#5}
   /SetVariables(Particles.VisibilityTag=&MakeCustomIntFromBool(Bool=$Particles.Is Active, True INT=1, False INT=0)) {isSetParameters=true, module=SetVariables_#6}
   &ParticleState(DeltaTime:NiagaraFloat=&MakeCustomFloatFromBool(Bool=$Particles.Is Active, True Float=$Engine.DeltaTime, False Float=0)) {module=ParticleState001}
   /SetVariables(Particles.Velocity=&MakeCustomVectorFromBool(
     Bool=$Particles.Is Active
     True Vector=&Multiply_VectorByFloat(
      Vector=&RandomRangeVector(Minimum=<-1250 0 1200>, Maximum=<-1000 0 700>)
      Float=&FloatFromCurve(FloatCurve=DataInterface() {propertyValues {Curve {keys [
           {interpMode=RCIM_Cubic, tangentMode=RCTM_User, time=0, value=0.2}
           {arriveTangent=0.3, interpMode=RCIM_Cubic, leaveTangent=0.3, time=0.2, value=0.4}
           {interpMode=RCIM_Cubic, tangentMode=RCTM_User, time=1, value=0.5}
          ]}}})
     )
     False Vector=<0 0 0>
    )) {isSetParameters=true, module=SetVariables_#7}
   &UpdateMeshOrientation(
    Rotation Rate=&MakeCustomFloatFromBool(Bool=$Particles.Is Active, True Float=0, False Float=1)
    Rotation Vector=&MakeVector(
     X=&Cosine(
      Normalized Angle=$Emitter.Age
      Period=&RandomRangeFloat(Minimum=1, Maximum=0.3)
      Scale=&UniformAOrBFloat(
       A=0
       B=&RandomRangeFloat(Minimum=0.7, Maximum=0.5)
       Distribution Weight=&RandomRangeFloat(Minimum=0.5, Maximum=0.7)
       Evaluation Type=::NewEnumerator0
       Recalculate Random Each Loop=true
       Override Seed=false
      )
      Bias=0
     )
     Y=0
     Z=0
    )
   ) {enabled=false}
   &AvoidDistanceFieldSurfaces_GPU(
    Active:NiagaraBool=$Particles.Is Active
    Avoid Nearest Surface=true
    Avoid Oncoming Surfaces=true
    Nearest Surface Avoidance Strength=25000
    Obstactle Avoidance Strength=25000
   ) {module=AvoidDistanceFieldSurfaces_GPU001}
   &Collision(
    Age Colliding Particles=false
    Collision Enabled=&SetBoolByFloatComparison(A=$Particles.Age, B=0.2)
    Enable Rest State=false
    GPU Collision Type=::NewEnumerator2
    Kill On Collision=15
    KillOnCollision=true
    Module.EnableMaxCollisionCount=false
    Module.Randomize Collision Normal Vector=false
    Particle Radius Scale=9
    Radius Calculation Type=::NewEnumerator1
    Restitution=0
   ) {module=Collision001}
   &SolveForcesAndVelocity(Clamp Velocity=true, Limit Acceleration=true)
   &ScaleMeshSize(Scale Factor=&VectorFromFloat(Value=&FloatFromCurve(FloatCurve=DataInterface() {propertyValues {Curve {keys [
          {tangentMode=RCTM_User, time=0, value=1}
          {arriveTangent=-0.0405, interpMode=RCIM_Cubic, leaveTangent=-0.0405, tangentMode=RCTM_User, time=0.85, value=1}
          {arriveTangent=-8.0089, interpMode=RCIM_Cubic, leaveTangent=-8.0089, tangentMode=RCTM_User, time=1, value=0}
         ]}}})))
   &KillParticlesInVolume(
    Box Size:Vector3f=<3000 15555 3000>
    Kill Shape=::NewEnumerator1
    Volume Origin:Vector3f=&AddVectorToPosition(Position=<88272 9968 4316>, Vector=<0 0 0>)
   ) {module=KillParticlesInVolume002}
   &DynamicMaterialParameters(
    Index 0 Param 2=&RandomRangeFloat(Minimum=2.1, Maximum=2.7)
    Module.Index0Float0Write=false
    Module.Index0Float1Write=true
    Module.Index0Float2Write=false
    Module.Index0Float3Write=false
   )
  ]
  rendererClasses [&NiagaraMeshRendererProperties, &NiagaraSpriteRendererProperties]
  renderers [
   {rendererClass=&NiagaraMeshRendererProperties, rendererIndex=0}
   {rendererClass=&NiagaraMeshRendererProperties, rendererIndex=1}
   {rendererClass=&NiagaraMeshRendererProperties, rendererIndex=2}
   {rendererClass=&NiagaraSpriteRendererProperties, rendererIndex=3}
  ]
 }
}
summary {
 systemName=NS_FNM_PumpkinPass_Crows_IdleToFlight
 emitters [
  {
   bEnabled=false
   emitterName=Crows_SceneDepth
   simTarget=GPUComputeSim
   rendererClasses [&NiagaraMeshRendererProperties, &NiagaraSpriteRendererProperties]
  }
  {
   emitterName=Crows_SceneDepth001
   simTarget=GPUComputeSim
   rendererClasses [&NiagaraMeshRendererProperties, &NiagaraSpriteRendererProperties]
  }
 ]
 userVariables [
  User.RadiusMult {type=NiagaraFloat, defaultValue=1, description=""}
  User.SpawnRateMult {type=NiagaraFloat, defaultValue=1, description=""}
  User.TransformScale {type=Vector3f, description="", defaultValue {struct=Vector3f, value=<1 1 0>}}
 ]
}
